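# Niagara system for the Quick Burst bass emote prop, adding electric arcs, glows, lights and sparks pinned to skeletal mesh sockets.
# Mesh emitters animate bursts, string electricity and knob sparking through dynamic material parameters.
# Ribbon_Ends_Control writes sampled end positions into a shared array, which Ribbon_Strikes reads to build noise-bent lightning ribbons.
# User.Strum scales the Strum emitter's looping bursts and two Burst spawn rates.
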NiagaraSystem /BRCosmetics/Animation/Props/Gadgets_QuickBurst_Bass/FX/NS_Emote_QuickBurst_BassFX
dependencies {
 usedDataInterfaces [
  &NiagaraDataInterfaceCurve
  &NiagaraDataInterfaceArrayFloat
  &NiagaraDataInterfaceArrayInt32
  &NiagaraDataInterfaceArrayFloat3
  &NiagaraDataInterfaceSkeletalMesh
  /Script/Niagara.NiagaraDataInterfaceArrayFloat4
  &NiagaraDataInterfaceArrayPosition
  /Script/Niagara.NiagaraDataInterfaceSimpleCounter
 ]
 usedDynamicInputs []
 usedModules [
  &self:SystemSpawnScript.SystemScriptSource.SystemScriptGraph.NiagaraNodeAssignment_0.SetVariables_#1
  &self:SystemSpawnScript.SystemScriptSource.SystemScriptGraph.NiagaraNodeAssignment_1.SetVariables_#2
  &self:SystemSpawnScript.SystemScriptSource.SystemScriptGraph.NiagaraNodeAssignment_2.SetVariables_#3
  &SystemState
 ]
}
emitters {
 Burst {
  emitterUpdateScript [
   &EmitterState()
   &SpawnRate(SpawnRate=7, Use Spawn Probability=false)
   &SpawnRate(SpawnGroup=1, SpawnRate=7, Use Spawn Probability=false) {module=SpawnRate001}
   &SpawnRate(SpawnGroup=2, SpawnRate=&RandomRangeFloat(Minimum=1, Maximum=7), Use Spawn Probability=false) {module=SpawnRate002}
   &SpawnRate(SpawnGroup=3, SpawnRate=&Multiply_Float(A=7, B=$User.Strum), Use Spawn Probability=false) {module=SpawnRate003}
   &SpawnRate(SpawnGroup=4, SpawnRate=&Multiply_Float(A=7, B=$User.Strum), Use Spawn Probability=false) {module=SpawnRate004}
  ]
  particleSpawnScript [
   /SetVariables(Particles.VisibilityTag=&MakeCustomIntFromBool(
     Bool=&SetBoolByIntComparison(A=&SpawnGroup(), B=1, Comparison Type=::NewEnumerator4)
     True INT=1
     False INT=0
    )) {isSetParameters=true}
   &InitializeParticle(
    Color=&RandomRangeLinearColor(Minimum=<rgba 0.8 0.8 0.8 0.2>, Maximum=<rgba 1 1 1 0.5>)
    Lifetime=&MaskFloatBySpawnGroup(
     Masked Float=&RandomRangeFloat(Minimum=0.3, Maximum=0.5)
     Passthrough Float=&MaskFloatBySpawnGroup(
      Masked Float=&RandomRangeFloat(Minimum=0.1, Maximum=0.3)
      Passthrough Float=&MaskFloatBySpawnGroup(
       Masked Float=&RandomRangeFloat(Minimum=0.2, Maximum=0.4)
       Passthrough Float=&RandomRangeFloat(Minimum=0.5, Maximum=0.6)
       Spawn Group Mask=4
      )
      Spawn Group Mask=3
     )
    )
    Mesh Scale Mode=::NewEnumerator3
    Mesh Uniform Scale=&MaskFloatBySpawnGroup(
     Masked Float=&RandomRangeFloat(Minimum=0.04, Maximum=0.06)
     Passthrough Float=&RandomRangeFloat(Minimum=0.05, Maximum=0.1)
     Spawn Group Mask=2
    )
    UsePositionOffset=false
    Write Material Random=true
   )
   &AddVelocity(
    Cone Angle=&MaskFloatBySpawnGroup(Masked Float=45, Passthrough Float=30, Spawn Group Mask=2)
    Cone Axis=&SelectVectorFromArray(
     Direct Array Index=&SpawnGroup()
     Vector Selection Array=DataInterface() {propertyValues {FloatData [<0 0 1>, <0 0 1>, <-0.2 -1 0.25>, <-1 -0.3 0>, <-1 -0.3 0>]}}
    )
    Module.UseConeFalloff=false
    Module.UseVelDistribution=false
    Override Seed=false
    Velocity Mode=::NewEnumerator2
    Velocity Speed=0.1
   )
   &InitialMeshOrientation(Mesh Axis To Orient=$Particles.Velocity, Mesh Orientation Mode=::NewEnumerator2, Orientation Vector=<0 0 1>, Use Rotation Vector=false)
  ]
  particleUpdateScript [
   &ParticleState()
   &SkeletalMeshLocation(
    Bone / Socket Index=&SelectIntFromArray(Direct Array Index=&SpawnGroup(), Int Selection Array=DataInterface() {propertyValues {IntData [0,0,7,14,14]}})
    MaskSpawnGroup=false
    Mesh Sampling Type=::NewEnumerator5
    Sampled Position Offset=&SelectVectorFromArray(
     Direct Array Index=&SpawnGroup()
     Vector Selection Array=DataInterface() {propertyValues {FloatData [<0 2 -6>, <0 2 -6>, <0 0 0>, <0 0 0>, <0 0 0>]}}
    )
    Skeletal Mesh=$System.Skeletal Mesh
    Socket Sampling Mode=::NewEnumerator5
   )
   &ScaleMeshSize(Scale Factor=&RandomRangeVector(Minimum=<1 1 2>, Maximum=<1.3 1.3 2>)) {module=ScaleMeshSize001}
   &ScaleMeshSize(Scale Factor=&MakeVector(
     X=1
     Y=1
     Z=&FloatFromCurve(FloatCurve=DataInterface() {propertyValues {Curve {keys [
          {interpMode=RCIM_Cubic, leaveTangent=0.0966, tangentMode=RCTM_Break, time=0, value=1}
          {arriveTangent=0.0951, interpMode=RCIM_Cubic, leaveTangent=-0.1745, tangentMode=RCTM_Break, time=1, value=1.07}
         ]}}})
    ))
   &DynamicMaterialParameters(
    Index 0 Param 4=&MakeCustomFloatFromBool(
     Bool=false
     True Float=5
     False Float=&SelectFloatFromArray(
      Array Sampling Mode=::NewEnumerator0
      Direct Array Index=?
      Float Selection Array=DataInterface() {propertyValues {FloatData [3,0,10,3,1,9,5]}}
      Random Seed=0
     )
    )
    Index 1 Param 1=1
    Index 1 Param 2=2
    Index 1 Param 3=&FloatFromCurve(FloatCurve=DataInterface() {propertyValues {Curve {keys [{interpMode=RCIM_Cubic, tangentMode=RCTM_Break, time=0.5, value=-0.5}, <at 1 0.1>]}}})
    Index 2 Param 1=&Multiply_Float(
     A=&FloatFromCurve(
      FloatCurve=DataInterface() {propertyValues {Curve {keys [{interpMode=RCIM_Cubic, leaveTangent=8.1085, tangentMode=RCTM_Break, time=0.8568, value=1}, {tangentMode=RCTM_User, time=1, value=0}]}}}
      Scale Curve=0.5
     )
     B=&Divide_Float(
      A=&MakeFloatFromVector(Vector=$Particles.Scale, Channel=::NewEnumerator0)
      B=&RandomRangeFloat(Minimum=10, Maximum=15)
     )
    )
    Index 2 Param 2=&FloatFromCurve(
     FloatCurve=DataInterface() {propertyValues {Curve {keys [
         {interpMode=RCIM_Cubic, tangentMode=RCTM_Break, time=0.7, value=0}
         {arriveTangent=-0.0859, interpMode=RCIM_Cubic, tangentMode=RCTM_Break, time=1, value=-1}
        ]}}}
     Scale Curve=50
    )
    Index 2 Param 3=&Waveform(
     Global Input Position=&RandomRangeFloat(Minimum=0, Maximum=15)
     Global Phase=$Particles.Age
     "Waveform [1]"=::NewEnumerator12
     "[1] Phase Scale"=2
     "[1] Frequency"=1
     "[1] Amplitude Min/Max"=<0 1>
    )
    Index 2 Param 4=25
    Index 3 Param 1=&FloatFromCurve(
     FloatCurve=DataInterface() {propertyValues {Curve {keys [
         {interpMode=RCIM_Cubic, leaveTangent=-0.0025, tangentMode=RCTM_Break, time=0, value=0}
         {arriveTangent=1.8528, interpMode=RCIM_Cubic, tangentMode=RCTM_Break, time=1, value=1}
        ]}}}
     Scale Curve=&RandomRangeFloat(Minimum=15, Maximum=25)
    )
    Index 3 Param 2=&FloatFromCurve(FloatCurve=DataInterface() {propertyValues {Curve {keys [
         {interpMode=RCIM_Cubic, leaveTangent=-0.0025, tangentMode=RCTM_Break, time=0, value=0.5}
         {arriveTangent=-0.0316, interpMode=RCIM_Cubic, tangentMode=RCTM_Break, time=1, value=-0.2}
        ]}}})
    Index 3 Param 3=&FloatFromCurve(FloatCurve=DataInterface() {propertyValues {Curve {keys [
         {interpMode=RCIM_Cubic, leaveTangent=1.1034, tangentMode=RCTM_Break, time=0, value=-1}
         {arriveTangent=2.2399, tangentMode=RCTM_User, time=0.5, value=0.125}
        ]}}})
    Index 3 Param 4=&FloatFromCurve(FloatCurve=DataInterface() {propertyValues {Curve {keys [
         {interpMode=RCIM_Cubic, leaveTangent=3.0601, tangentMode=RCTM_Break, time=0.5, value=-0.5}
         {arriveTangent=3.2053, tangentMode=RCTM_User, time=1, value=1}
        ]}}})
    Module.Index0Param0Write=true
    Module.Index0Param1Write=true
    Module.Index0Param2Write=true
    Module.Index0Param3Write=true
    Module.Index1Param0Write=true
    Module.Index1Param1Write=true
    Module.Index1Param2Write=true
    Module.Index1Param3Write=true
    Module.Index2Param0Write=true
    Module.Index2Param1Write=true
    Module.Index2Param2Write=true
    Module.Index2Param3Write=true
    Module.Index3Param0Write=true
    Module.Index3Param1Write=true
    Module.Index3Param2Write=true
    Module.Index3Param3Write=true
    Write Parameter Index 1=true
    Write Parameter Index 2=true
    Write Parameter Index 3=true
   )
   &SolveForcesAndVelocity(Clamp Velocity=false, Limit Acceleration=false)
   &ScaleColor(
    Scale Alpha=&FloatFromCurve(FloatCurve=DataInterface() {propertyValues {Curve {keys [
         {tangentMode=RCTM_User, time=0, value=1}
         {arriveTangent=-0.0405, interpMode=RCIM_Cubic, leaveTangent=-0.0405, tangentMode=RCTM_User, time=0.7009, value=1}
         {arriveTangent=-8.0089, interpMode=RCIM_Cubic, leaveTangent=-8.0089, tangentMode=RCTM_User, time=1, value=0}
        ]}}})
    Scale RGB=&Lerp_Vector(
     A=&VectorFromFloat(Value=&FloatFromCurve(FloatCurve=DataInterface() {propertyValues {Curve {keys [<at 0.2771 1>,<at 0.3786 1>]}}}, Scale Curve=20))
     B=&VectorFromFloat(Value=&RandomRangeFloat(Minimum=0.3, Maximum=2, Evaluation Type=::NewEnumerator1))
     Alpha=&FloatFromCurve(FloatCurve=DataInterface() {propertyValues {Curve {keys [
          {interpMode=RCIM_Cubic, leaveTangent=2.5294, tangentMode=RCTM_Break, time=0, value=0}
          {arriveTangent=-0.2166, interpMode=RCIM_Cubic, tangentMode=RCTM_Break, time=0.4, value=1}
         ]}}})
    )
    ScaleA=true
    ScaleRGB=true
   )
  ]
  rendererClasses [&NiagaraMeshRendererProperties]
  renderers [
   {rendererClass=&NiagaraMeshRendererProperties, rendererIndex=0}
   {rendererClass=&NiagaraMeshRendererProperties, rendererIndex=1}
   {rendererClass=&NiagaraMeshRendererProperties, rendererIndex=2}
  ]
 }
 Electric_Strings {
  emitterUpdateScript [
   &EmitterState()
   &SpawnRate(SpawnRate=24, Use Spawn Probability=false)
  ]
  particleSpawnScript [
   &PartitionParticles(Number Of Partitions=4, Particles Per Partition=1, Partition Mode:ENiagara_PartitionMode=::NewEnumerator1)
   &InitializeParticle(
    Color Mode=::NewEnumerator0
    Lifetime=0.25
    Mesh Scale Mode=::NewEnumerator3
    Mesh Uniform Scale=&RandomRangeFloat(Minimum=0.035, Maximum=0.1)
    UsePositionOffset=false
    Write Material Random=true
   )
   &CameraOffset(Camera Offset Amount=10)
  ]
  particleUpdateScript [
   &ParticleState()
   &SkeletalMeshLocation(
    Bone / Socket Index=&SelectIntFromArray(Direct Array Index=$Particles.PartitionParticles.Partition, Int Selection Array=DataInterface() {propertyValues {IntData [10,11,12,13]}})
    MaskSpawnGroup=false
    Mesh Sampling Type=::NewEnumerator5
    Skeletal Mesh=$System.Skeletal Mesh
    Socket Sampling Mode=::NewEnumerator5
   )
   &DynamicMaterialParameters(
    Index 0 Param 4=&SelectFloatFromArray(
     Array Sampling Mode=::NewEnumerator0
     Direct Array Index=?
     Float Selection Array=DataInterface() {propertyValues {FloatData [3,0,10,3,1,9,5]}}
     Random Seed=0
    )
    Index 1 Param 1=0.5
    Index 1 Param 2=2
    Index 1 Param 3=&FloatFromCurve(FloatCurve=DataInterface() {propertyValues {Curve {keys [{interpMode=RCIM_Cubic, tangentMode=RCTM_Break, time=0.5, value=-0.5}, <at 1 0.1>]}}})
    Index 1 Param 4=2
    Index 2 Param 1=&Multiply_Float(
     A=&FloatFromCurve(
      FloatCurve=DataInterface() {propertyValues {Curve {keys [{interpMode=RCIM_Cubic, leaveTangent=8.1085, tangentMode=RCTM_Break, time=0.8568, value=1}, {tangentMode=RCTM_User, time=1, value=0}]}}}
      Scale Curve=0.5
     )
     B=&Divide_Float(
      A=&MakeFloatFromVector(Vector=$Particles.Scale, Channel=::NewEnumerator0)
      B=&RandomRangeFloat(Minimum=10, Maximum=15)
     )
    )
    Index 2 Param 2=&FloatFromCurve(
     FloatCurve=DataInterface() {propertyValues {Curve {keys [
         {interpMode=RCIM_Cubic, tangentMode=RCTM_Break, time=0.7, value=0}
         {arriveTangent=-0.0859, interpMode=RCIM_Cubic, tangentMode=RCTM_Break, time=1, value=-1}
        ]}}}
     Scale Curve=50
    )
    Index 2 Param 3=&Waveform(
     Global Input Position=&RandomRangeFloat(Minimum=0, Maximum=15)
     Global Phase=$Particles.Age
     "Waveform [1]"=::NewEnumerator12
     "[1] Phase Scale"=2
     "[1] Frequency"=1
     "[1] Amplitude Min/Max"=<0 1>
    )
    Index 2 Param 4=25
    Index 3 Param 1=&FloatFromCurve(
     FloatCurve=DataInterface() {propertyValues {Curve {keys [
         {interpMode=RCIM_Cubic, leaveTangent=-0.0025, tangentMode=RCTM_Break, time=0, value=0}
         {arriveTangent=1.8528, interpMode=RCIM_Cubic, tangentMode=RCTM_Break, time=1, value=1}
        ]}}}
     Scale Curve=&RandomRangeFloat(Minimum=15, Maximum=25)
    )
    Index 3 Param 2=&FloatFromCurve(FloatCurve=DataInterface() {propertyValues {Curve {keys [
         {interpMode=RCIM_Cubic, leaveTangent=-0.0025, tangentMode=RCTM_Break, time=0, value=0.5}
         {arriveTangent=-0.0316, interpMode=RCIM_Cubic, tangentMode=RCTM_Break, time=1, value=-0.2}
        ]}}})
    Index 3 Param 3=&FloatFromCurve(FloatCurve=DataInterface() {propertyValues {Curve {keys [
         {interpMode=RCIM_Cubic, leaveTangent=1.1034, tangentMode=RCTM_Break, time=0, value=-1}
         {arriveTangent=2.2399, tangentMode=RCTM_User, time=0.5, value=0.125}
        ]}}})
    Index 3 Param 4=&FloatFromCurve(FloatCurve=DataInterface() {propertyValues {Curve {keys [
         {interpMode=RCIM_Cubic, leaveTangent=3.0601, tangentMode=RCTM_Break, time=0.5, value=-0.5}
         {arriveTangent=3.2053, tangentMode=RCTM_User, time=1, value=1}
        ]}}})
    Module.Index0Param0Write=true
    Module.Index0Param1Write=true
    Module.Index0Param2Write=true
    Module.Index0Param3Write=true
    Module.Index1Param0Write=true
    Module.Index1Param1Write=true
    Module.Index1Param2Write=true
    Module.Index1Param3Write=true
    Module.Index2Param0Write=true
    Module.Index2Param1Write=true
    Module.Index2Param2Write=true
    Module.Index2Param3Write=true
    Module.Index3Param0Write=true
    Module.Index3Param1Write=true
    Module.Index3Param2Write=true
    Module.Index3Param3Write=true
    Write Parameter Index 1=true
    Write Parameter Index 2=true
    Write Parameter Index 3=true
   )
  ]
  rendererClasses [&NiagaraMeshRendererProperties]
  renderers [
   {rendererClass=&NiagaraMeshRendererProperties, rendererIndex=0}
   {rendererClass=&NiagaraMeshRendererProperties, rendererIndex=1}
  ]
 }
 Glows {
  emitterUpdateScript [
   &EmitterState(
    Inactive Response=::NewEnumerator1
    Life Cycle Mode=::NewEnumerator1
    Loop Behavior=::NewEnumerator1
    Loop Duration Mode=::NewEnumerator1
    UseLoopDelay=false
   )
   &SpawnBurst_Instantaneous(Spawn Count=5, Use Spawn Probability=false, UseLoopCountLimit=true)
   &SpawnBurst_Instantaneous(Spawn Group=1, Use Spawn Probability=false, UseLoopCountLimit=true) {module=SpawnBurst_Instantaneous001}
  ]
  particleSpawnScript [
   /SetVariables(Particles.VisibilityTag=&SpawnGroup()) {isSetParameters=true}
   &PartitionParticles(Number Of Partitions=5)
   &InitializeParticle(
    Color=<rgba 0.1 0.4 1 0.3>
    Lifetime=0.8
    Position Offset=<0 0 15>
    Position Offset Coordinate Space=::Local
    Sprite Size Mode=::NewEnumerator3
    Uniform Sprite Size=&SelectFloatFromArray(
     Direct Array Index=$Particles.PartitionParticles.Partition
     Float Selection Array=DataInterface() {propertyValues {FloatData [30,20,5,5,20]}}
    )
    UsePositionOffset=true
    Write Material Random=true
   )
   &CameraOffset(Camera Offset Amount=5)
  ]
  particleUpdateScript [
   &ParticleState(Kill Particles When Lifetime Has Elapsed=false, Let Infinitely Lived Particles Die When Emitter Deactivates=true)
   &ScaleColor(
    Scale Alpha=&RandomRangeFloat(Minimum=0.5, Maximum=1, Evaluation Type=::NewEnumerator1)
    ScaleA=true
    ScaleRGB=true
   )
   &SkeletalMeshLocation(
    Bone / Socket Index=&MaskIntBySpawnGroup(
     Masked Int=&SelectIntFromArray(Direct Array Index=$Particles.PartitionParticles.Partition, Int Selection Array=DataInterface() {propertyValues {IntData [0,7,8,9,11]}})
     Passthrough Int=14
     Spawn Group Mask=0
    )
    MaskSpawnGroup=false
    Mesh Sampling Type=::NewEnumerator5
    Sampled Position Offset=&MaskVectorBySpawnGroup(
     Masked Vector=&SelectVectorFromArray(
      Direct Array Index=$Particles.PartitionParticles.Partition
      Vector Selection Array=DataInterface() {propertyValues {FloatData [<0 2 0>, <0 1 2>, <0 0 0>, <0 0 0>, <-2 0 5>]}}
     )
     Passthrough Vector=<0 15 0>
    )
    Skeletal Mesh=$System.Skeletal Mesh
    Socket Sampling Mode=::NewEnumerator5
   )
   /SetVariables(
    Particles.LightColor=&MultiplyLinearColor(
     Linear Color=$Particles.Color
     Scale Factor=&MakeFloat4FromVectorAndFloat(
      "Vector (RGB)"=&VectorFromFloat(Value=&Add_Float(A=50, B=&RandomRangeFloat(Minimum=0, Maximum=50, Evaluation Type=::NewEnumerator1)))
      "Float (Alpha)"=1
     )
    )
    Particles.LightExponent=4
   ) {isSetParameters=true}
  ]
  rendererClasses [&NiagaraSpriteRendererProperties, &NiagaraLightRendererProperties]
  renderers [
   {rendererClass=&NiagaraSpriteRendererProperties}
   {rendererClass=&NiagaraLightRendererProperties, rendererIndex=1}
  ]
 }
 Ribbon_Ends_Control {
  emitterUpdateScript [
   &EmitterState(Life Cycle Mode=::NewEnumerator1, Loop Behavior=::NewEnumerator1, Loop Duration Mode=::NewEnumerator1, UseLoopDelay=false)
   &SpawnRate(SpawnRate=14, Use Spawn Probability=false)
   &self:Reset_Counter(Simple Counter:NiagaraDataInterfaceSimpleCounter=$System.Simple Counter) {module=Reset_Counter}
  ]
  particleSpawnScript [&InitializeParticle(Lifetime Max=0.4, Lifetime Min=0.2, Lifetime Mode=::NewEnumerator1, UsePositionOffset=false, Write Material Random=true)]
  particleUpdateScript [
   &ParticleState()
   &self:ClearArray(
    Simple Counter:NiagaraDataInterfaceSimpleCounter=$System.Simple Counter
    Vector 4 Array:NiagaraDataInterfaceArrayFloat4=$System.End Position Array
   ) {module=ClearArray}
   &SkeletalMeshLocation(
    Bone / Socket Index=6
    MaskSpawnGroup=true
    Mesh Sampling Type=::NewEnumerator5
    Orientation Sampling=::NewEnumerator1
    Position Sampling=::NewEnumerator1
    Sampled Position Offset=&RandomRangeVector(Minimum=<0 -2 -30>, Maximum=<0 2 20>)
    Skeletal Mesh=$System.Skeletal Mesh
    Socket Sampling Mode=::NewEnumerator5
   )
   &self:RibbonArraySetup(
    End Position:NiagaraPosition=$Particles.SkeletalMeshLocation.SampledPosition
    End Position Array:NiagaraDataInterfaceArrayFloat4=$System.End Position Array
    Random Array:NiagaraDataInterfaceArrayFloat=$System.Random Float Array
    Random Value:NiagaraFloat=&RandomRangeFloat(Minimum=0, Maximum=1)
   ) {module=RibbonArraySetup}
  ]
  rendererClasses [&NiagaraSpriteRendererProperties]
  renderers [{rendererClass=&NiagaraSpriteRendererProperties}]
 }
 Ribbon_Strikes {
  emitterUpdateScript [
   &EmitterState(Life Cycle Mode=::NewEnumerator1, Loop Behavior=::NewEnumerator1, Loop Duration Mode=::NewEnumerator1, UseLoopDelay=false)
   &SpawnBurst_Instantaneous(Spawn Count=30, Use Spawn Probability=false, UseLoopCountLimit=false)
  ]
  particleSpawnScript [&InitializeParticle(
    Color=&MakeLinearColorFromVectorAndFloat("Vector (RGB)"=<1 1 1>, "Float (Alpha)"=0.5)
    Lifetime=5
    Ribbon Width=8
    Ribbon Width Mode=::NewEnumerator1
    Sprite Size Mode=::NewEnumerator3
    UsePositionOffset=false
    Write Material Random=true
   )]
  particleUpdateScript [
   &ParticleState(Kill Particles When Lifetime Has Elapsed=false, Let Infinitely Lived Particles Die When Emitter Deactivates=true)
   &self:SetupRibbonLogic(
    End Postion Array:NiagaraDataInterfaceArrayFloat4=$System.End Position Array
    NumofRibbon=6
    ParticlePerRibbon=5
    Random Array:NiagaraDataInterfaceArrayFloat=$System.Random Float Array
    Skeletal Mesh:NiagaraDataInterfaceSkeletalMesh=$System.Skeletal Mesh
    Starting Socket Index:NiagaraInt32=&RandomRangeInt(Minimum=0, Maximum=1)
   ) {module=SetupRibbonLogic}
   &self:RibbonBending(
    BigNoiseMask:NiagaraFloat=&FloatFromCurve(
     FloatCurve=DataInterface() {propertyValues {Curve {keys [
         {arriveTangent=3.5637, interpMode=RCIM_Cubic, leaveTangent=3.5637, leaveTangentWeight=0.2004, tangentMode=RCTM_User, time=0, value=0}
         {arriveTangentWeight=0.2638, interpMode=RCIM_Cubic, leaveTangentWeight=0.2638, tangentMode=RCTM_User, time=0.5, value=1}
         {arriveTangent=-3.2085, arriveTangentWeight=0.1773, interpMode=RCIM_Cubic, tangentMode=RCTM_Break, time=1, value=0}
        ]}}}
     CurveIndex=$Particles.RibbonLinkOrder
    )
    Offset:NiagaraFloat=&FloatFromCurve(
     FloatCurve=DataInterface() {propertyValues {Curve {keys [
         {interpMode=RCIM_Cubic, leaveTangentWeight=0.5631, tangentMode=RCTM_User, time=0, value=0}
         {arriveTangentWeight=0.1957, interpMode=RCIM_Cubic, leaveTangentWeight=0.1957, tangentMode=RCTM_User, time=0.5, value=0.2}
         <at 1 0 RCIM_Cubic>
        ]}}}
     CurveIndex=$Particles.RibbonLinkOrder
     Scale Curve=&FloatFromCurve(
      FloatCurve=DataInterface() {propertyValues {Curve {keys [
          {arriveTangent=2.2423, interpMode=RCIM_Cubic, leaveTangent=2.2423, tangentMode=RCTM_User, time=0, value=0.2}
          <at 0.5 1 RCIM_Cubic>
          {arriveTangent=-1.6083, interpMode=RCIM_Cubic, leaveTangent=-1.6083, tangentMode=RCTM_User, time=1, value=0.2}
         ]}}}
      CurveIndex=$Particles.Aged Elements
      Scale Curve=&Multiply_Float(A=$Particles.Random, B=-25)
     )
    )
    YBigNoise:NiagaraFloat=&Waveform(
     Global Input Position=&Multiply_Float(A=$Particles.Random, B=10)
     Global Phase=&Add_Float(
      A=&Multiply_Float(A=$Particles.RibbonLinkOrder, B=2)
      B=&Multiply_Float(A=$Emitter.Age, B=1.77)
     )
     "Waveform [1]"=::NewEnumerator12
     "[1] Frequency"=1
     "[1] Amplitude Min/Max"=<-1 1>
     "[1] Scale"=15
    )
    ZBigNoise:NiagaraFloat=&Waveform(
     Global Input Position=&Multiply_Float(A=$Particles.Random, B=10)
     Global Phase=&Add_Float(
      A=&Multiply_Float(A=$Particles.RibbonLinkOrder, B=3)
      B=&Multiply_Float(A=$Emitter.Age, B=2)
     )
     "Waveform [1]"=::NewEnumerator12
     "[1] Frequency"=1
     "[1] Amplitude Min/Max"=<-1 0>
     "[1] Scale"=15
    )
   ) {module=RibbonBending}
   &ScaleRibbonWidth(Ribbon Width Scale=&FloatFromCurve(
     FloatCurve=DataInterface() {propertyValues {Curve {keys [
         {interpMode=RCIM_Cubic, tangentMode=RCTM_Break, time=0, value=0.6}
         {interpMode=RCIM_Cubic, tangentMode=RCTM_Break, time=0.1, value=0.4}
         {interpMode=RCIM_Cubic, tangentMode=RCTM_Break, time=0.9, value=0.4}
         <at 1 1>
        ]}}}
     CurveIndex=$Particles.RibbonLinkOrder
    ))
   &ScaleRibbonWidth(Ribbon Width Scale=&RemapRange(Input Value To Remap=$Particles.Random, Input Min=0, Input Max=1, Output Min=0.7, Output Max=2)) {module=ScaleRibbonWidth001}
   /SetVariables(Particles.RibbonWidth2:NiagaraFloat=&Multiply_Float(A=$Particles.RibbonWidth, B=3)) {isSetParameters=true}
   &DynamicMaterialParameters(
    Index 1 Param 1=$Particles.Random
    Index 1 Param 2=&FloatFromCurve(FloatCurve=DataInterface() {propertyValues {Curve {keys [<at 0 0>,<at 0.3 1>]}}}, CurveIndex=$Particles.Aged Elements)
    Index 1 Param 3=&FloatFromCurve(
     FloatCurve=DataInterface() {propertyValues {Curve {keys [<at 0 0>,<at 0.5 1>,<at 0.6 1>,<at 1 0>]}}}
     CurveIndex=$Particles.Aged Elements
     Scale Curve=0
    )
    Module.Index1Param0Write=true
    Module.Index1Param1Write=true
    Module.Index1Param2Write=true
    Module.Index1Param3Write=true
    Write Parameter Index 0=false
    Write Parameter Index 1=true
   )
   &ScaleColor(
    Scale Alpha=&FloatFromCurve(
     FloatCurve=DataInterface() {propertyValues {Curve {keys [
         {tangentMode=RCTM_User, time=0, value=1}
         {arriveTangent=-0.0405, interpMode=RCIM_Cubic, leaveTangent=-0.0405, tangentMode=RCTM_User, time=0.7009, value=1}
         {arriveTangent=-8.0089, interpMode=RCIM_Cubic, leaveTangent=-8.0089, tangentMode=RCTM_User, time=1, value=0}
        ]}}}
     CurveIndex=$Particles.Aged Elements
    )
    Scale RGB=&VectorFromFloat(Value=&FloatFromCurve(
      FloatCurve=DataInterface() {propertyValues {Curve {keys [<at 0 7>,<at 0.3 1>]}}}
      CurveIndex=$Particles.Aged Elements
      Scale Curve=&RandomRangeFloat(Minimum=0.4, Maximum=1, Evaluation Type=::NewEnumerator1)
     ))
    ScaleA=true
    ScaleRGB=true
   )
   /SetVariables(Particles.Soft Ribbon:LinearColor=&MultiplyLinearColor(
     Linear Color=$Particles.Color
     Scale Factor=&MakeFloat4FromVectorAndFloat(
      "Vector (RGB)"=&VectorFromFloat(Value=&FloatFromCurve(FloatCurve=DataInterface() {propertyValues {Curve {keys [<at 0.7 1>,<at 1 5>]}}}, CurveIndex=$Particles.Aged Elements, Scale Curve=1.5))
      "Float (Alpha)"=1
     )
    )) {isSetParameters=true}
   &CameraOffset(Camera Offset Amount=1)
   /SetVariables(
    Particles.Glow_Color:LinearColor=<rgba 0.1 0.4 1 0.1>
    Particles.Sprite_Color:LinearColor=&MultiplyLinearColor(
     Linear Color=<rgba 0.1 0.4 1 1>
     Scale Factor=&MakeFloat4FromVectorAndFloat(
      "Vector (RGB)"=<2 2 2>
      "Float (Alpha)"=&FloatFromCurve(
       FloatCurve=DataInterface() {propertyValues {Curve {keys [
           <at 0.29 0 RCIM_Constant>
           <at 0.3 1>
           {arriveTangent=-6.0095, interpMode=RCIM_Cubic, leaveTangent=-2.1556, tangentMode=RCTM_Break, time=0.6, value=0}
          ]}}}
       CurveIndex=$Particles.Aged Elements
       Scale Curve=&Multiply_Float(
        A=&RemapRange(Input Value To Remap=$Particles.Random, Input Min=0, Input Max=1, Output Min=0.3, Output Max=0.9)
        B=&Multiply_Float(
         A=&FNTimeOfDayScalar(Night Scale=0.7, Morning Scale=1, Evening Scale=1)
         B=&FN_FrontEndFloatModifier(Gameplay=1, Front End=0.5)
        )
       )
      )
     )
    )
    Particles.Sprite_Sizing:Vector2f=&Multiply_Vector2D_ByFloat(
     Float=&FloatFromCurve(
      FloatCurve=DataInterface() {propertyValues {Curve {keys [<at 0.3 1>,<at 0.5 2>]}}}
      CurveIndex=$Particles.Aged Elements
      Scale Curve=&RemapRange(Input Value To Remap=$Particles.Random, Input Min=0, Input Max=1, Output Min=0.3, Output Max=1)
     )
     Vector2D=<6 6>
    )
   ) {isSetParameters=true}
  ]
  rendererClasses [&NiagaraRibbonRendererProperties, &NiagaraSpriteRendererProperties]
  renderers [
   {rendererClass=&NiagaraRibbonRendererProperties}
   {rendererClass=&NiagaraRibbonRendererProperties, rendererIndex=1}
   {rendererClass=&NiagaraSpriteRendererProperties, rendererIndex=2}
   {rendererClass=&NiagaraSpriteRendererProperties, rendererIndex=3}
  ]
 }
 Strum {
  emitterUpdateScript [
   &EmitterState(Life Cycle Mode=::NewEnumerator1, Loop Duration=0.1, UseLoopDelay=false)
   &SpawnBurst_Instantaneous(
    Spawn Count=&Multiply_Int(A=1, B=&MakeIntFromFloat(Float=$User.Strum))
    Use Spawn Probability=false
    UseLoopCountLimit=false
   )
   &SpawnBurst_Instantaneous(
    Spawn Count=&Multiply_Int(A=1, B=&MakeIntFromFloat(Float=$User.Strum))
    Spawn Group=1
    Use Spawn Probability=false
    UseLoopCountLimit=false
   ) {module=SpawnBurst_Instantaneous001}
   &SpawnBurst_Instantaneous(
    Spawn Count=&Multiply_Int(
     A=&RandomRangeInt(Minimum=5, Maximum=20)
     B=&MakeIntFromFloat(Float=$User.Strum)
    )
    Spawn Group=2
    Use Spawn Probability=false
    UseLoopCountLimit=false
   ) {module=SpawnBurst_Instantaneous002}
  ]
  particleSpawnScript [
   /SetVariables(Particles.VisibilityTag=&SpawnGroup()) {isSetParameters=true}
   &InitializeParticle(
    Color=&MaskLinearColorBySpawnGroup(
     Masked Linear Color=<rgba 0.1 0.4 1 5>
     Passthrough Linear Color=&MaskLinearColorBySpawnGroup(Masked Linear Color=<rgba 0.1 0.4 1 1>, Passthrough Linear Color=<rgba 1.5 8.625 30 1>, Spawn Group Mask=1)
    )
    Lifetime=&MaskFloatBySpawnGroup(Masked Float=&RandomRangeFloat(Minimum=0.05, Maximum=0.3), Passthrough Float=0.25, Spawn Group Mask=2)
    Mesh Scale Mode=::NewEnumerator3
    Mesh Uniform Scale=0.1
    Sprite Size=&MaskVector2dBySpawnGroup(
     Masked Vector2D=<80 80>
     Passthrough Vector2D=&RandomRangeVector2D(Minimum=<1 1>, Maximum=<1 5>)
     Spawn Group Mask=0
    )
    Sprite Size Mode=::NewEnumerator4
    UsePositionOffset=false
    Write Material Random=true
   )
   &SkeletalMeshLocation(
    Bone / Socket Index=14
    MaskSpawnGroup=false
    Mesh Sampling Type=::NewEnumerator5
    Position Sampling=::NewEnumerator2
    Skeletal Mesh=$System.Skeletal Mesh
    Socket Sampling Mode=::NewEnumerator5
   ) {module=SkeletalMeshLocation001}
   &ShapeLocation(
    MaskSpawnGroup=true
    Override Seed=false
    Shape Origin=$Particles.SkeletalMeshLocation001.SampledPosition
    Spawn Group Mask=2
    Sphere Radius=1
    UseOwnerScale=true
   )
   &AddVelocity(
    Module.Use Falloff Distance=false
    Override Seed=false
    Velocity Mode=::NewEnumerator1
    Velocity Origin=$StackContext.Position
    Velocity Speed=&RandomRangeFloat(Minimum=100, Maximum=200)
   )
   &CameraOffset(Camera Offset Amount=&MaskFloatBySpawnGroup(Masked Float=5, Passthrough Float=0))
  ]
  particleUpdateScript [
   &ParticleState()
   &SkeletalMeshLocation(
    Bone / Socket Index=14
    MaskSpawnGroup=false
    Mesh Sampling Type=::NewEnumerator5
    Position Sampling=::NewEnumerator2
    Skeletal Mesh=$System.Skeletal Mesh
    Socket Sampling Mode=::NewEnumerator5
   )
   &UpdateMeshOrientation(Rotation Rate=0.2, Rotation Vector=<1 1 1>)
   &GravityForce(Gravity=&MaskVectorBySpawnGroup(Masked Vector=<0 0 -980>, Passthrough Vector=<0 0 0>, Spawn Group Mask=2))
   &Drag(Drag=&RandomRangeFloat(Minimum=1, Maximum=6), Use Linear Drag=true, Use Rotational Drag=true)
   &CurlNoiseForce(Noise Frequency=10, Noise Quality / Cost=::NewEnumerator0, Noise Strength=500, Pan Noise=true, Pan Noise Field=<0 0 1>)
   &SolveForcesAndVelocity(Clamp Velocity=false, Limit Acceleration=false)
   &ScaleSpriteSizeBySpeed(Max Scale Factor=<1 2>, Min Scale Factor=<0.5 1>)
   &DynamicMaterialParameters(
    Index 0 Param 1=&MaskFloatBySpawnGroup(
     Masked Float=0.5
     Passthrough Float=&FloatFromCurve(FloatCurve=DataInterface() {propertyValues {Curve {keys [<at 0 1>,<at 0.2 0>]}}})
    )
    Index 0 Param 2=1
    Module.Index0Param0Write=true
    Module.Index0Param1Write=true
    Module.Index0Param2Write=true
    Module.Index0Param3Write=false
   )
  ]
  rendererClasses [&NiagaraSpriteRendererProperties, &NiagaraMeshRendererProperties]
  renderers [
   {rendererClass=&NiagaraSpriteRendererProperties}
   {rendererClass=&NiagaraMeshRendererProperties, rendererIndex=1}
   {rendererClass=&NiagaraSpriteRendererProperties, rendererIndex=2}
  ]
 }
 VoltKnobSparking {
  emitterUpdateScript [
   &EmitterState()
   &SpawnRate(SpawnRate=&RandomRangeFloat(Minimum=1, Maximum=20), Use Spawn Probability=false)
   &SpawnRate(SpawnGroup=1, SpawnRate=&RandomRangeFloat(Minimum=1, Maximum=10), Use Spawn Probability=false) {module=SpawnRate001}
   &SpawnRate(SpawnGroup=2, SpawnRate=&RandomRangeFloat(Minimum=2, Maximum=20), Use Spawn Probability=false) {module=SpawnRate002}
  ]
  particleSpawnScript [
   &InitializeParticle(
    Color Maximum=<rgba 0.18 0.18 0.18 1>
    Color Minimum=<rgba 0.075 0.075 0.075 0.7>
    Color Mode=::NewEnumerator2
    Lifetime Max=0.5
    Lifetime Min=0.1
    Lifetime Mode=::NewEnumerator1
    Mesh Scale Mode=::NewEnumerator3
    Mesh Uniform Scale=&MaskFloatBySpawnGroup(
     Masked Float=&RandomRangeFloat(Minimum=0.075, Maximum=0.125)
     Passthrough Float=&MaskFloatBySpawnGroup(
      Masked Float=0.25
      Passthrough Float=&RandomRangeFloat(Minimum=0.075, Maximum=0.125)
      Spawn Group Mask=1
     )
    )
    UsePositionOffset=false
    Write Material Random=true
   )
   &InitialMeshOrientation(
    Mesh Axis To Orient=<0 1 0>
    Mesh Orientation Mode=::NewEnumerator2
    Orientation Vector=&MaskVectorBySpawnGroup(Masked Vector=<0 0 1>, Passthrough Vector=<0.5 0 1>)
    Rotation=&RandomRangeVector(Minimum=<0 0 0>, Maximum=<0 0 1>)
    Use Rotation Vector=true
   )
   &CameraOffset(Camera Offset Amount=&MaskFloatBySpawnGroup(Masked Float=2, Passthrough Float=1))
  ]
  particleUpdateScript [
   &ParticleState()
   &SkeletalMeshLocation(
    Bone / Socket Index=&MaskIntBySpawnGroup(
     Masked Int=&RandomRangeInt(Minimum=8, Maximum=9)
     Passthrough Int=&MaskIntBySpawnGroup(Masked Int=7, Passthrough Int=&RandomRangeInt(Minimum=0, Maximum=1), Spawn Group Mask=1)
     Spawn Group Mask=0
    )
    MaskSpawnGroup=false
    Mesh Sampling Type=::NewEnumerator5
    Orientation Sampling=::NewEnumerator1
    Sampled Position Offset=&MaskVectorBySpawnGroup(
     Masked Vector=<0.5 0 0>
     Passthrough Vector=&MaskVectorBySpawnGroup(Masked Vector=<0.5 3 1>, Passthrough Vector=<0 0 0>, Spawn Group Mask=1)
    )
    Skeletal Mesh=$System.Skeletal Mesh
    Socket Sampling Mode=::NewEnumerator5
   )
   &ScaleMeshSize(Scale Factor=&VectorFromFloat(Value=&FloatFromCurve(FloatCurve=DataInterface() {propertyValues {Curve {keys [
          {arriveTangent=1.9115, interpMode=RCIM_Cubic, leaveTangent=1.9115, tangentMode=RCTM_User, time=0, value=0}
          {arriveTangent=0.0224, interpMode=RCIM_Cubic, leaveTangent=0.0224, tangentMode=RCTM_User, time=1, value=1}
         ]}}})))
   &DynamicMaterialParameters(
    Index 0 Param 4=13
    Index 1 Param 1=&FloatFromCurve(FloatCurve=DataInterface() {propertyValues {Curve {keys [<at 0 0.5>,<at 1 1>]}}})
    Index 1 Param 2=2
    Index 1 Param 3=&FloatFromCurve(FloatCurve=DataInterface() {propertyValues {Curve {keys [{interpMode=RCIM_Cubic, tangentMode=RCTM_Break, time=0.5, value=-0.5}, <at 1 0.1>]}}})
    Index 2 Param 1=&Multiply_Float(
     A=&FloatFromCurve(
      FloatCurve=DataInterface() {propertyValues {Curve {keys [{interpMode=RCIM_Cubic, leaveTangent=8.1085, tangentMode=RCTM_Break, time=0.8568, value=1}, {tangentMode=RCTM_User, time=1, value=0}]}}}
      Scale Curve=0.25
     )
     B=&Divide_Float(
      A=&MakeFloatFromVector(Vector=$Particles.Scale, Channel=::NewEnumerator0)
      B=&RandomRangeFloat(Minimum=10, Maximum=15)
     )
    )
    Index 2 Param 2=&FloatFromCurve(
     FloatCurve=DataInterface() {propertyValues {Curve {keys [
         {interpMode=RCIM_Cubic, tangentMode=RCTM_Break, time=0.7, value=0}
         {arriveTangent=-0.0859, interpMode=RCIM_Cubic, tangentMode=RCTM_Break, time=1, value=-1}
        ]}}}
     Scale Curve=0
    )
    Index 2 Param 3=&Waveform(
     Global Input Position=&RandomRangeFloat(Minimum=0, Maximum=15)
     Global Phase=$Particles.Age
     "Waveform [1]"=::NewEnumerator12
     "[1] Phase Scale"=2
     "[1] Frequency"=1
     "[1] Amplitude Min/Max"=<0 1>
    )
    Index 2 Param 4=25
    Index 3 Param 1=&FloatFromCurve(
     FloatCurve=DataInterface() {propertyValues {Curve {keys [
         {interpMode=RCIM_Cubic, leaveTangent=-0.0025, tangentMode=RCTM_Break, time=0, value=0}
         {arriveTangent=1.8528, interpMode=RCIM_Cubic, tangentMode=RCTM_Break, time=1, value=1}
        ]}}}
     Scale Curve=&RandomRangeFloat(Minimum=0, Maximum=0)
    )
    Index 3 Param 2=&FloatFromCurve(
     FloatCurve=DataInterface() {propertyValues {Curve {keys [
         {interpMode=RCIM_Cubic, leaveTangent=-0.0025, tangentMode=RCTM_Break, time=0, value=0.5}
         {arriveTangent=-0.0316, interpMode=RCIM_Cubic, tangentMode=RCTM_Break, time=1, value=-0.2}
        ]}}}
     Scale Curve=0
    )
    Index 3 Param 3=&FloatFromCurve(FloatCurve=DataInterface() {propertyValues {Curve {keys [
         {interpMode=RCIM_Cubic, leaveTangent=1.1034, tangentMode=RCTM_Break, time=0, value=-1}
         {arriveTangent=2.2399, tangentMode=RCTM_User, time=0.5, value=0.125}
        ]}}})
    Index 3 Param 4=&FloatFromCurve(FloatCurve=DataInterface() {propertyValues {Curve {keys [
         {interpMode=RCIM_Cubic, leaveTangent=1.8089, tangentMode=RCTM_Break, time=0.3, value=-0.5}
         {arriveTangent=2.0757, interpMode=RCIM_Cubic, tangentMode=RCTM_Break, time=0.9, value=1}
        ]}}})
    Module.Index0Param0Write=true
    Module.Index0Param1Write=true
    Module.Index0Param2Write=true
    Module.Index0Param3Write=true
    Module.Index1Param0Write=true
    Module.Index1Param1Write=true
    Module.Index1Param2Write=true
    Module.Index1Param3Write=true
    Module.Index2Param0Write=true
    Module.Index2Param1Write=true
    Module.Index2Param2Write=true
    Module.Index2Param3Write=true
    Module.Index3Param0Write=true
    Module.Index3Param1Write=true
    Module.Index3Param2Write=true
    Module.Index3Param3Write=true
    Write Parameter Index 1=true
    Write Parameter Index 2=true
    Write Parameter Index 3=true
   )
   &ScaleColor(
    Scale Alpha=&FloatFromCurve(FloatCurve=DataInterface() {propertyValues {Curve {keys [
         {tangentMode=RCTM_User, time=0, value=1}
         {arriveTangent=-4.3986, interpMode=RCIM_Cubic, leaveTangent=-4.3986, tangentMode=RCTM_User, time=0.35, value=1}
         {arriveTangent=0.1205, interpMode=RCIM_Cubic, leaveTangent=0.1205, tangentMode=RCTM_User, time=1, value=0}
        ]}}})
    Scale RGB=&VectorFromFloat(Value=&Lerp_Float(
      A=20
      B=&RandomRangeFloat(Minimum=-0.5, Maximum=2, Evaluation Type=::NewEnumerator1)
      Alpha=&FloatFromCurve(FloatCurve=DataInterface() {propertyValues {Curve {keys [
           {arriveTangent=1.0951, interpMode=RCIM_Cubic, leaveTangent=1.0951, leaveTangentWeight=0.3923, tangentMode=RCTM_User, time=0, value=0}
           {arriveTangentWeight=0.3923, interpMode=RCIM_Cubic, tangentMode=RCTM_User, time=1, value=1}
          ]}}})
     ))
    ScaleA=true
    ScaleRGB=true
   )
   /SetVariables(Particles.Glow Mesh Color:LinearColor=&MultiplyLinearColor(
     Linear Color=$Particles.Color
     Scale Factor=&MakeFloat4FromVectorAndFloat("Vector (RGB)"=<1 1 1>, "Float (Alpha)"=5)
    )) {isSetParameters=true}
  ]
  rendererClasses [&NiagaraMeshRendererProperties]
  renderers [
   {rendererClass=&NiagaraMeshRendererProperties, rendererIndex=0}
   {rendererClass=&NiagaraMeshRendererProperties, rendererIndex=1}
  ]
 }
}
summary {
 systemName=NS_Emote_QuickBurst_BassFX
 emitters [
  {emitterName=VoltKnobSparking, rendererClasses [&NiagaraMeshRendererProperties]}
  {
   emitterName=Ribbon_Strikes
   rendererClasses [&NiagaraRibbonRendererProperties, &NiagaraSpriteRendererProperties]
  }
  {emitterName=Ribbon_Ends_Control, rendererClasses [&NiagaraSpriteRendererProperties]}
  {emitterName=Glows, rendererClasses [&NiagaraSpriteRendererProperties, &NiagaraLightRendererProperties]}
  {emitterName=Burst, rendererClasses [&NiagaraMeshRendererProperties]}
  {emitterName=Electric_Strings, rendererClasses [&NiagaraMeshRendererProperties]}
  {emitterName=Strum, rendererClasses [&NiagaraSpriteRendererProperties, &NiagaraMeshRendererProperties]}
 ]
 userVariables [
  User.bIsFrontEnd {type=NiagaraBool}
  User.bIsFrontEndPreview {type=NiagaraBool}
  User.Strum {type=NiagaraFloat, defaultValue=1, description=""}
 ]
}
